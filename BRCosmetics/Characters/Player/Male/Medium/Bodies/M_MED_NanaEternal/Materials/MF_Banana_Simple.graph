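# Material function for the NanaEternal character body that restyles the output of the shared weapon-skin Banana function.
# It adds a Fresnel glow from a squared color to the emissive, blending from scaled at center to full at edges.
# Base color becomes a squared tint masked by rim and facing Fresnel terms, with roughness and specular overridden.
# A UseWPO_Only static switch bypasses these overrides and outputs the Banana result unchanged.
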MaterialFunction /BRCosmetics/Characters/Player/Male/Medium/Bodies/M_MED_NanaEternal/Materials/MF_Banana_Simple
graph {nodes [
  n1 = MaterialFunctionCall!(Gmap, Material=FunctionInput() {inputName=Material, inputType=FunctionInput_MaterialAttributes}) {materialFunction=/Game/Weapons/WeaponSkins/MaterialFunction/Banana}
  n2 = GetMaterialAttributes(None=@n1.Result) {attributeGetTypes [#1]}
  n3 = VectorParameter() {group=Banana, parameterName=FresnelColor02, defaultValue=<rgba 0.0045 1 0.7005 1>}
  n4 = Multiply(A=@n3.RGB, B=@n3.RGB)
  n5 = Multiply(A=@n4, B=ScalarParameter() {defaultValue=0.1, group=Banana, parameterName=EmissiveColorOverride})
  n6 = Fresnel() {baseReflectFraction=0, exponent=1}
  n7 = VectorParameter() {group=Control, parameterName=BaseColorOverride_Color, defaultValue=<rgba 0 0.5906 0.1363 1>}
  n8 = Multiply(A=@n7.RGB, B=@n7.RGB)
  n9 = Fresnel(ExponentIn=ScalarParameter() {defaultValue=2, group=TwosideCull, parameterName=Fresnel01}) {baseReflectFraction=0, exponent=1}
  n10 = Power(
   Base=OneMinus(None=Fresnel() {baseReflectFraction=0, exponent=1})
   Exp=ScalarParameter() {defaultValue=4, group=TwosideCull, parameterName=Fresnel02}
  )
  n11 = SetMaterialAttributes!(
   MaterialAttributes=@n2.MaterialAttributes
   EmissiveColor=Add(A=@n2.EmissiveColor, B=LinearInterpolate(A=@n5, B=@n4, Alpha=@n6) {constA=1})
   BaseColor=Multiply(A=Saturate(None=@n8), B=Max(A=@n9, B=@n10))
   Roughness=ScalarParameter() {defaultValue=1, group=Control, parameterName=RoughnessOverride}
   Specular=ScalarParameter() {group=Control, parameterName=SpecularOverride}
  ) {attributeSetTypes [
    #1
    #2
    #3
    #4
   ]}
  FunctionOutput(None=StaticSwitchParameter(True=@n1.Result, False=@n11) {group=Control, parameterName=UseWPO_Only})
 ]}
