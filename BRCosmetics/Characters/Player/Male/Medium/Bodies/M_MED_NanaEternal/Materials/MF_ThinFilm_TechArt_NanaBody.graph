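# Material function for the NanaEternal body that adds an iridescent thin-film tint to the base color of incoming material attributes.
# A scaled, warped Fresnel input, optionally offset by a noise channel, serves as coordinates into a spectrum texture.
# The result is added to base color by Fresnel falloff, optionally attenuated by roughness.
# A channel-selected Alpha mask blends the effect in, and a static switch bypasses it entirely.
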MaterialFunction /BRCosmetics/Characters/Player/Male/Medium/Bodies/M_MED_NanaEternal/Materials/MF_ThinFilm_TechArt_NanaBody
graph {nodes [
  n1 = GetMaterialAttributes(None=FunctionInput() {inputName=Material Attributes, inputType=FunctionInput_MaterialAttributes, sortPriority=1}) {attributeGetTypes [#1, #2]}
  n2 = TextureSampleParameter2D(UVs=Multiply(A=TextureCoordinate(), B=ScalarParameter() {defaultValue=2, parameterName=UvFilmOffsetTiling})) {parameterName=ThinFilmUVOffsetTexture, samplerSource=SSM_Wrap_WorldGroupSettings, texture=/Game/Global/Textures/Default/Noise/3xPerlin_M}
  n3 = ChannelMaskParameter(None=@n2.RGB) {group=06_Thin Film, maskChannel=Green, parameterName=ThinFilmUVOffsetChannel, defaultValue=<rgba 0 1 0 0>}
  n4 = Multiply(A=Multiply(A=@n3.R) {constB=0.1}, B=ScalarParameter() {defaultValue=0.2, group=06_Thin Film, parameterName=ThinFilmUVOffsetIntensity})
  n5 = FunctionInput() {inputName=Fresnel, inputType=FunctionInput_Scalar, sortPriority=2}
  n6 = ScalarParameter() {defaultValue=0.5, group=06_Thin Film, parameterName=ThinFilmWarp, sortPriority=4}
  n7 = Add(
   A=Multiply(A=ScalarParameter() {defaultValue=1, group=06_Thin Film, parameterName=ThinFilmScale, sortPriority=5}, B=Power(Base=@n5, Exp=@n6))
   B=ScalarParameter() {group=06_Thin Film, parameterName=ThinFilmOffset}
  )
  n8 = TextureSampleParameter2D(UVs=StaticSwitchParameter(True=Add(A=@n4, B=@n7), False=@n7) {parameterName=UseThinFilmUVOffset}) {
   group=06_Thin Film
   parameterName=Thin Film Texture
   samplerSource=SSM_Wrap_WorldGroupSettings
   sortPriority=1
   texture=/Game/Global/Textures/Default/Thinfilm/T_ThinFilm_Spectrum_COLOR
  }
  n9 = Multiply(A=@n8.RGB, B=ScalarParameter() {defaultValue=0.5, group=06_Thin Film, parameterName=ThinFilmIntensity, sortPriority=2})
  n10 = Power(Base=@n5, Exp=ScalarParameter() {defaultValue=2, group=06_Thin Film, parameterName=ThinFilmExponent, sortPriority=3})
  n11 = LinearInterpolate(
   A=@n10
   B=Multiply(A=LinearInterpolate(Alpha=@n1.Roughness) {constA=1, constB=0.01}, B=@n10)
   Alpha=ScalarParameter() {group=06_Thin Film, parameterName=RoughnessInfluence}
  )
  n12 = SetMaterialAttributes!(
   MaterialAttributes=NamedRerouteUsage()
   BaseColor=LinearInterpolate(A=@n1.BaseColor, B=Add(A=@n1.BaseColor, B=@n9), Alpha=Saturate(None=@n11))
  ) {attributeSetTypes [#2]}
  n13 = NamedRerouteUsage()
  n14 = ChannelMaskParameter(None=FunctionInput() {bUsePreviewValueAsDefault=true, inputName=Alpha, sortPriority=3}) {group=05_Tech Art Masks, parameterName=ThinFilmMaskChannel}
  FunctionOutput(None=StaticSwitchParameter(True=BlendMaterialAttributes(A=@n13, B=@n12, Alpha=@n14.R) {vertexAttributeBlendType=UseA}, False=@n13) {group=04_Tech Art_RE, parameterName=UseThinFilmRE})
  NamedRerouteDeclaration(None=@n1.MaterialAttributes) {name=Material Attributes, nodeColor=<rgba 0.0781 0 1 1>}
 ]}
