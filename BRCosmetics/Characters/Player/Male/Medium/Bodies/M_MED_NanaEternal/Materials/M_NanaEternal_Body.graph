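# Masked two-sided body material for the NanaEternal character, layered between the shared 2023 character PreFX and PostFX functions.
# A two-sided sign plus the NanaMask texture drives opacity, hiding front faces outside the mask while keeping backfaces, with ShadowReplace keeping shadows solid.
# Roughness and specular take scalar overrides, and a static switch optionally adds a SubUV flipbook face.
# A Fresnel-based color chain is left unconnected.
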Material /BRCosmetics/Characters/Player/Male/Medium/Bodies/M_MED_NanaEternal/Materials/M_NanaEternal_Body
graph {
 properties {
  bUseMaterialAttributes=true
  bUsedWithMeshDeformer=true
  bUsedWithMorphTargets=true
  bUsedWithNiagaraMeshParticles=true
  bUsedWithSkeletalMesh=true
  blendMode=BLEND_Masked
  pixelDepthOffsetMode=PDOM_Legacy
  shadingModel=MSM_FromMaterialExpression
  twoSided=true
  usedShadingModels=Default Lit | Subsurface | Clear Coat | Subsurface Profile | Cloth
 }
 nodes [
  n1 = MaterialFunctionCall() {materialFunction=&MF_CharacterParent_2023_PreFX}
  n2 = GetMaterialAttributes(None=@n1.Result) {attributeGetTypes [#1]}
  n3 = TextureSampleParameter2D() {group=TwosideCull, parameterName=NanaMask, samplerType=SAMPLERTYPE_Masks, texture=&T_BlackMask}
  n4 = Sign(None=TwoSidedSign())
  n5 = ShadowReplace(Default=LinearInterpolate(Alpha=Saturate(None=@n4)) {constA=1, constB=0}, Shadow=Constant() {r=1})
  n6 = SetMaterialAttributes!(
   MaterialAttributes=@n2.MaterialAttributes
   Opacity Mask=Saturate(None=Add(A=@n3.R, B=@n5))
   Base Color=@n2.MaterialAttributes
   Roughness=ScalarParameter() {defaultValue=1, group=Control, parameterName=RoughnessOverride}
   Specular=ScalarParameter() {group=Control, parameterName=SpecularOverride}
  ) {attributeSetTypes [
    #2
    #1
    #3
    #4
   ]}
  MaterialFunctionCall(
   Material=StaticSwitchParameter(
    True=MaterialFunctionCall!(Material=@n6) {materialFunction=&/MF_SubUVTextures_NanaEternal}.Result
    False=@n6
   ) {group=Control, parameterName=UseFlipbookFace}
   M=@n1.M
   Pre Skinned Local Position=@n1.M
   Distance Blend=@n1.M
   Fresnel=@n1.M
  ) {materialFunction=&MF_CharacterParent_2023_PostFX}
  n7 = Fresnel() {baseReflectFraction=0, exponent=1}
  n8 = Max(
   A=Fresnel(ExponentIn=ScalarParameter() {defaultValue=2, group=TwosideCull, parameterName=Fresnel01}) {baseReflectFraction=0, exponent=1}
   B=Power(Base=OneMinus(None=@n7), Exp=ScalarParameter() {defaultValue=4, group=TwosideCull, parameterName=Fresnel02})
  )
  Multiply(A=@n2.Base Color, B=@n8)
  Power(Base=@n8) {constExponent=3}
 ]
}
