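# Control Rig for the NanaEternal medium male player body that corrects deformation and keeps dynamic limb bones aimed at their end points.
# A shared DeformRig function runs first, driving corrective deform bones such as upperarm, elbow, wrist, groin, glute and knee.
# Each side then aims hip/knee bones at each other and fibula, tibia, radius and ulna at end bones via from-two-vectors quaternions.
# Forearm bones aim along negated X, with set-rotation weight -1.
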ControlRigBlueprint /BRCosmetics/Characters/Player/Male/Medium/Bodies/M_MED_NanaEternal/Meshes/M_MED_NanaEternal_CtrlRig
graphs [
 RigVMFunctionLibrary {links [], nodes []}
 RigVMModel {nodes [
   n1 = &RigUnit_BeginExecution()
   FN_DeformRig_FNC = RigVMFunctionReferenceNode(
    DeformRig_ Curve="(deformRig_l_upperarm_up_alpha_72_#1=False,deformRig_r_upperarm_up_alpha_73_#2=False,deformRig_l_upperarm_fr_alpha_74_#3=False,deformRig_r_upperarm_fr_alpha_75_CB51B8 ...(truncated)"
    DeformRig_Bool="(upperarm_1_#4=True,elbow_3_#5=True,wrist_7_#6=True,groin_8_#7=True,glute_17_#8=True,knee_9_40C408324CA261 ...(truncated)"
    DeformRig_Debug=(upperarm_debug_18_#4=False,wrist_debug_20_#6=False,lat_debug_26_#9=False,groinGlute_debug_27_#7=False)
    DeformRig_Value="(upperarm_up_18_#10=(Rotation=(X=0.0,Y=-0.0,Z=0.0,W=1.0),Translation=(X=0.0,Y=0.0,Z=8.0),Scale3D=(X=0.70,Y=1.0,Z=1.0)),upperarm_fwd_fr_19_#11=(Rotation=(X=0.0,Y=-0.0,Z=0.0,W=1.0),Translation=(X=0.0, ...(truncated)"
    ExecuteContext=@n1.ExecutePin
   )
   n2 = &RigUnit_GetTransform(Item=(Type=Bone,Name="dyn_hipToKnee_r"), bInitial=False)
   n3 = /Script/RigVM.RigVMFunction_MathQuaternionGetAxis(Axis=X, Quaternion=@n2.Transform.Rotation) {is=struct RigVMUnitNode, template="Axis::Execute(in Quaternion,in Axis,out Result)"}
   n4 = &RigUnit_GetTransform(Item=(Type=Bone,Name="dyn_kneeToHip_r"), bInitial=False)
   n5 = /Script/RigVM.RigVMFunction_MathVectorSub(
    B=@n2.Transform.Translation
    A=/Script/RigVM.RigVMFunction_MathTransformTransformVector(Location=(X=0,Y=0,Z=0), Transform=@n4.Transform) {is=struct RigVMUnitNode}.Result
   ) {is=struct RigVMUnitNode, template="Subtract::Execute(in A,in B,out Result)"}
   > Set Transform = &RigUnit_SetRotation(
    Item=(Type=Bone,Name="dyn_hipToKnee_r")
    Space=GlobalSpace
    Weight=1
    bInitial=False
    Value=/Script/RigVM.RigVMFunction_MathQuaternionMul(
     B=@n2.Transform.Rotation
     A=/Script/RigVM.RigVMFunction_MathQuaternionFromTwoVectors(A=@n3.Result, B=@n5.Result) {is=struct RigVMUnitNode, template="FromTwoVectors::Execute(in A,in B,out Result)"}.Result
    ) {is=struct RigVMUnitNode, template="Multiply::Execute(in A,in B,out Result)"}.Result
   )
   n6 = &RigVMFunction_MathQuaternionGetAxis(Axis=X, Quaternion=@n4.Transform.Rotation) {template="Axis::Execute(in Quaternion,in Axis,out Result)"}
   n7 = &RigVMFunction_MathVectorSub(
    B=@n4.Transform.Translation
    A=&RigVMFunction_MathTransformTransformVector(Location=(X=0,Y=0,Z=0), Transform=@n2.Transform).Result
   ) {template="Subtract::Execute(in A,in B,out Result)"}
   > Set Transform_1 = &RigUnit_SetRotation(
    Item=(Type=Bone,Name="dyn_kneeToHip_r")
    Space=GlobalSpace
    Weight=1
    bInitial=False
    Value=&RigVMFunction_MathQuaternionMul(
     B=@n4.Transform.Rotation
     A=&RigVMFunction_MathQuaternionFromTwoVectors(A=@n6.Result, B=@n7.Result) {template="FromTwoVectors::Execute(in A,in B,out Result)"}.Result
    ) {template="Multiply::Execute(in A,in B,out Result)"}.Result
   )
   n8 = &RigUnit_GetTransform(Item=(Type=Bone,Name="dyn_fibula_Base_r"), bInitial=False)
   n9 = &RigVMFunction_MathQuaternionGetAxis(Axis=X, Quaternion=@n8.Transform.Rotation) {template="Axis::Execute(in Quaternion,in Axis,out Result)"}
   n10 = &RigVMFunction_MathVectorSub(
    B=@n8.Transform.Translation
    A=&RigVMFunction_MathTransformTransformVector(Location=(X=0,Y=0,Z=0), Transform=&RigUnit_GetTransform(Item=(Type=Bone,Name="dyn_fibulaEnd_r"), bInitial=False).Transform).Result
   ) {template="Subtract::Execute(in A,in B,out Result)"}
   > Set Transform_2 = &RigUnit_SetRotation(
    Item=(Type=Bone,Name="dyn_fibula_Base_r")
    Space=GlobalSpace
    Weight=1
    bInitial=False
    Value=&RigVMFunction_MathQuaternionMul(
     B=@n8.Transform.Rotation
     A=&RigVMFunction_MathQuaternionFromTwoVectors(A=@n9.Result, B=@n10.Result) {template="FromTwoVectors::Execute(in A,in B,out Result)"}.Result
    ) {template="Multiply::Execute(in A,in B,out Result)"}.Result
   )
   n11 = &RigUnit_GetTransform(Item=(Type=Bone,Name="dyn_tibia_Base_r"), bInitial=False)
   n12 = &RigVMFunction_MathQuaternionGetAxis(Axis=X, Quaternion=@n11.Transform.Rotation) {template="Axis::Execute(in Quaternion,in Axis,out Result)"}
   n13 = &RigVMFunction_MathVectorSub(
    B=@n11.Transform.Translation
    A=&RigVMFunction_MathTransformTransformVector(Location=(X=0,Y=0,Z=0), Transform=&RigUnit_GetTransform(Item=(Type=Bone,Name="dyn_tibiaEnd_r"), bInitial=False).Transform).Result
   ) {template="Subtract::Execute(in A,in B,out Result)"}
   > Set Transform_3 = &RigUnit_SetRotation(
    Item=(Type=Bone,Name="dyn_tibia_Base_r")
    Space=GlobalSpace
    Weight=1
    bInitial=False
    Value=&RigVMFunction_MathQuaternionMul(
     B=@n11.Transform.Rotation
     A=&RigVMFunction_MathQuaternionFromTwoVectors(A=@n12.Result, B=@n13.Result) {template="FromTwoVectors::Execute(in A,in B,out Result)"}.Result
    ) {template="Multiply::Execute(in A,in B,out Result)"}.Result
   )
   n14 = &RigUnit_GetBoneTransform(Bone=dyn_radius_r)
   n15 = /Script/RigVM.RigVMFunction_MathVectorNegate(Value=&RigVMFunction_MathQuaternionGetAxis(Axis=X, Quaternion=@n14.Transform.Rotation) {template="Axis::Execute(in Quaternion,in Axis,out Result)"}.Result) {is=struct RigVMUnitNode, template="Negate::Execute(in Value,out Result)"}
   n16 = &RigVMFunction_MathVectorSub(
    B=@n14.Transform.Translation
    A=&RigVMFunction_MathTransformTransformVector(Location=(X=0,Y=0,Z=0), Transform=&RigUnit_GetBoneTransform(Bone=dyn_radiusEnd_r).Transform).Result
   ) {template="Subtract::Execute(in A,in B,out Result)"}
   > /Script/ControlRig.RigUnit_SetBoneRotation(
    Bone=dyn_radius_r
    Space=GlobalSpace
    Weight=-1
    bPropagateToChildren=True
    Rotation=&RigVMFunction_MathQuaternionMul(
     B=@n14.Transform.Rotation
     A=&RigVMFunction_MathQuaternionFromTwoVectors(A=@n15.Result, B=@n16.Result) {template="FromTwoVectors::Execute(in A,in B,out Result)"}.Result
    ) {template="Multiply::Execute(in A,in B,out Result)"}.Result
   ) {is=struct RigVMUnitNode}
   n17 = &RigUnit_GetBoneTransform(Bone=dyn_ulna_r)
   n18 = &RigVMFunction_MathVectorNegate(Value=&RigVMFunction_MathQuaternionGetAxis(Axis=X, Quaternion=@n17.Transform.Rotation) {template="Axis::Execute(in Quaternion,in Axis,out Result)"}.Result) {template="Negate::Execute(in Value,out Result)"}
   n19 = &RigVMFunction_MathVectorSub(
    B=@n17.Transform.Translation
    A=&RigVMFunction_MathTransformTransformVector(Location=(X=0,Y=0,Z=0), Transform=&RigUnit_GetBoneTransform(Bone=dyn_ulnaEnd_r).Transform).Result
   ) {template="Subtract::Execute(in A,in B,out Result)"}
   > &RigUnit_SetBoneRotation(
    Bone=dyn_ulna_r
    Space=GlobalSpace
    Weight=-1
    bPropagateToChildren=True
    Rotation=&RigVMFunction_MathQuaternionMul(
     B=@n17.Transform.Rotation
     A=&RigVMFunction_MathQuaternionFromTwoVectors(A=@n18.Result, B=@n19.Result) {template="FromTwoVectors::Execute(in A,in B,out Result)"}.Result
    ) {template="Multiply::Execute(in A,in B,out Result)"}.Result
   )
   n20 = &RigUnit_GetTransform(Item=(Type=Bone,Name="dyn_hipToKnee_l"), bInitial=False)
   n21 = &RigVMFunction_MathQuaternionGetAxis(Axis=X, Quaternion=@n20.Transform.Rotation) {template="Axis::Execute(in Quaternion,in Axis,out Result)"}
   n22 = &RigUnit_GetTransform(Item=(Type=Bone,Name="dyn_kneeToHip_l"), bInitial=False)
   n23 = &RigVMFunction_MathVectorSub(
    B=@n20.Transform.Translation
    A=&RigVMFunction_MathTransformTransformVector(Location=(X=0,Y=0,Z=0), Transform=@n22.Transform).Result
   ) {template="Subtract::Execute(in A,in B,out Result)"}
   > Set Transform_4 = &RigUnit_SetRotation(
    Item=(Type=Bone,Name="dyn_hipToKnee_l")
    Space=GlobalSpace
    Weight=1
    bInitial=False
    Value=&RigVMFunction_MathQuaternionMul(
     B=@n20.Transform.Rotation
     A=&RigVMFunction_MathQuaternionFromTwoVectors(A=@n21.Result, B=@n23.Result) {template="FromTwoVectors::Execute(in A,in B,out Result)"}.Result
    ) {template="Multiply::Execute(in A,in B,out Result)"}.Result
   )
   n24 = &RigVMFunction_MathQuaternionGetAxis(Axis=X, Quaternion=@n22.Transform.Rotation) {template="Axis::Execute(in Quaternion,in Axis,out Result)"}
   n25 = &RigVMFunction_MathVectorSub(
    B=@n22.Transform.Translation
    A=&RigVMFunction_MathTransformTransformVector(Location=(X=0,Y=0,Z=0), Transform=@n20.Transform).Result
   ) {template="Subtract::Execute(in A,in B,out Result)"}
   > Set Transform_5 = &RigUnit_SetRotation(
    Item=(Type=Bone,Name="dyn_kneeToHip_l")
    Space=GlobalSpace
    Weight=1
    bInitial=False
    Value=&RigVMFunction_MathQuaternionMul(
     B=@n22.Transform.Rotation
     A=&RigVMFunction_MathQuaternionFromTwoVectors(A=@n24.Result, B=@n25.Result) {template="FromTwoVectors::Execute(in A,in B,out Result)"}.Result
    ) {template="Multiply::Execute(in A,in B,out Result)"}.Result
   )
   n26 = &RigUnit_GetTransform(Item=(Type=Bone,Name="dyn_fibula_Base_l"), bInitial=False)
   n27 = &RigVMFunction_MathQuaternionGetAxis(Axis=X, Quaternion=@n26.Transform.Rotation) {template="Axis::Execute(in Quaternion,in Axis,out Result)"}
   n28 = &RigVMFunction_MathVectorSub(
    B=@n26.Transform.Translation
    A=&RigVMFunction_MathTransformTransformVector(Location=(X=0,Y=0,Z=0), Transform=&RigUnit_GetTransform(Item=(Type=Bone,Name="dyn_fibulaEnd_l"), bInitial=False).Transform).Result
   ) {template="Subtract::Execute(in A,in B,out Result)"}
   > Set Transform_6 = &RigUnit_SetRotation(
    Item=(Type=Bone,Name="dyn_fibula_Base_l")
    Space=GlobalSpace
    Weight=1
    bInitial=False
    Value=&RigVMFunction_MathQuaternionMul(
     B=@n26.Transform.Rotation
     A=&RigVMFunction_MathQuaternionFromTwoVectors(A=@n27.Result, B=@n28.Result) {template="FromTwoVectors::Execute(in A,in B,out Result)"}.Result
    ) {template="Multiply::Execute(in A,in B,out Result)"}.Result
   )
   n29 = &RigUnit_GetTransform(Item=(Type=Bone,Name="dyn_tibia_Base_l"), bInitial=False)
   n30 = &RigVMFunction_MathQuaternionGetAxis(Axis=X, Quaternion=@n29.Transform.Rotation) {template="Axis::Execute(in Quaternion,in Axis,out Result)"}
   n31 = &RigVMFunction_MathVectorSub(
    B=@n29.Transform.Translation
    A=&RigVMFunction_MathTransformTransformVector(Location=(X=0,Y=0,Z=0), Transform=&RigUnit_GetTransform(Item=(Type=Bone,Name="dyn_tibiaEnd_l"), bInitial=False).Transform).Result
   ) {template="Subtract::Execute(in A,in B,out Result)"}
   > Set Transform_7 = &RigUnit_SetRotation(
    Item=(Type=Bone,Name="dyn_tibia_Base_l")
    Space=GlobalSpace
    Weight=1
    bInitial=False
    Value=&RigVMFunction_MathQuaternionMul(
     B=@n29.Transform.Rotation
     A=&RigVMFunction_MathQuaternionFromTwoVectors(A=@n30.Result, B=@n31.Result) {template="FromTwoVectors::Execute(in A,in B,out Result)"}.Result
    ) {template="Multiply::Execute(in A,in B,out Result)"}.Result
   )
   n32 = &RigUnit_GetBoneTransform(Bone=dyn_radius_l)
   n33 = &RigVMFunction_MathVectorNegate(Value=&RigVMFunction_MathQuaternionGetAxis(Axis=X, Quaternion=@n32.Transform.Rotation) {template="Axis::Execute(in Quaternion,in Axis,out Result)"}.Result) {template="Negate::Execute(in Value,out Result)"}
   n34 = &RigVMFunction_MathVectorSub(
    B=@n32.Transform.Translation
    A=&RigVMFunction_MathTransformTransformVector(Location=(X=0,Y=0,Z=0), Transform=&RigUnit_GetBoneTransform(Bone=dyn_radiusEnd_l).Transform).Result
   ) {template="Subtract::Execute(in A,in B,out Result)"}
   > &RigUnit_SetBoneRotation(
    Bone=dyn_radius_l
    Space=GlobalSpace
    Weight=-1
    bPropagateToChildren=True
    Rotation=&RigVMFunction_MathQuaternionMul(
     B=@n32.Transform.Rotation
     A=&RigVMFunction_MathQuaternionFromTwoVectors(A=@n33.Result, B=@n34.Result) {template="FromTwoVectors::Execute(in A,in B,out Result)"}.Result
    ) {template="Multiply::Execute(in A,in B,out Result)"}.Result
   )
   n35 = &RigUnit_GetBoneTransform(Bone=dyn_ulna_l)
   n36 = &RigVMFunction_MathVectorNegate(Value=&RigVMFunction_MathQuaternionGetAxis(Axis=X, Quaternion=@n35.Transform.Rotation) {template="Axis::Execute(in Quaternion,in Axis,out Result)"}.Result) {template="Negate::Execute(in Value,out Result)"}
   n37 = &RigVMFunction_MathVectorSub(
    B=@n35.Transform.Translation
    A=&RigVMFunction_MathTransformTransformVector(Location=(X=0,Y=0,Z=0), Transform=&RigUnit_GetBoneTransform(Bone=dyn_ulnaEnd_l).Transform).Result
   ) {template="Subtract::Execute(in A,in B,out Result)"}
   > &RigUnit_SetBoneRotation(
    Bone=dyn_ulna_l
    Space=GlobalSpace
    Weight=-1
    bPropagateToChildren=True
    Rotation=&RigVMFunction_MathQuaternionMul(
     B=@n35.Transform.Rotation
     A=&RigVMFunction_MathQuaternionFromTwoVectors(A=@n36.Result, B=@n37.Result) {template="FromTwoVectors::Execute(in A,in B,out Result)"}.Result
    ) {template="Multiply::Execute(in A,in B,out Result)"}.Result
   )
   &RigVMFunction_MathVectorNegate(Value=@n12.Result) {template="Negate::Execute(in Value,out Result)"}
   &RigVMFunction_MathVectorNegate(Value=@n30.Result) {template="Negate::Execute(in Value,out Result)"}
   &RigVMFunction_MathVectorNegate(Value=@n9.Result) {template="Negate::Execute(in Value,out Result)"}
   &RigVMFunction_MathVectorNegate(Value=@n27.Result) {template="Negate::Execute(in Value,out Result)"}
  ]}
]
hierarchy [
 root {loc=<0 0 0>, rot=<0 0 0 1>}
 pelvis {parent=root, loc=<0 -0.7071 93.1191>, rot=<0.0127 0.707 -0.0127 -0.707>}
 spine_01 {parent=pelvis, loc=<6.7494 -0.6388 0>, rot=<0 0 0.034 -0.9994>}
 spine_02 {parent=spine_01, loc=<7.2518 -0.035 0>, rot=<0 0 0.0238 -0.9997>}
 spine_03 {parent=spine_02, loc=<6.0104 0 0>, rot=<0 0 0.0149 -0.9999>}
 spine_04 {parent=spine_03, loc=<8.1024 0 0>, rot=<0 0 -0.0132 -0.9999>}
 spine_05 {parent=spine_04, loc=<10.0146 0 0>, rot=<0 0 -0.0015 -1>}
 clavicle_l {parent=spine_05, loc=<4.3575 -0.1466 -2.3223>, rot=<0.0465 -0.7289 -0.009 -0.6829>}
 upperarm_l {parent=clavicle_l, loc=<15.5111 0 0>, rot=<0.0226 -0.3234 -0.0088 -0.9459>}
 lowerarm_l {parent=upperarm_l, loc=<26.6514 0 0>, rot=<0.0391 -0.0099 -0.2453 -0.9686>}
 hand_l {parent=lowerarm_l, loc=<22.3491 0 0>, rot=<-0.7081 -0.0507 0.0975 -0.6975>}
 index_metacarpal_l {parent=hand_l, loc=<1.6535 0.0183 -2.96>, rot=<0.0158 -0.0124 -0.0173 -0.9996>}
 index_01_l {parent=index_metacarpal_l, loc=<8.3008 0.0357 -0.0524>, rot=<-0.0044 0.0485 0.0924 -0.9945>}
 index_02_l {parent=index_01_l, loc=<3.3587 -0.3004 0.0452>, rot=<-0.002 0.0027 0.3225 -0.9466>}
 index_03_l {parent=index_02_l, loc=<2.5578 -0.123 -0.0252>, rot=<-0.002 0.0067 0.2558 -0.9667>}
 middle_metacarpal_l {parent=hand_l, loc=<2.1856 -0.1725 -0.5984>, rot=<0.006 0.0168 -0.0329 -0.9993>}
 middle_01_l {parent=middle_metacarpal_l, loc=<7.3275 -0.2447 0.0129>, rot=<-0.0208 0.0042 0.0858 -0.9961>}
 middle_02_l {parent=middle_01_l, loc=<4.0123 0.1401 0.1175>, rot=<-0.0218 0.0258 0.3512 -0.9357>}
 middle_03_l {parent=middle_02_l, loc=<2.6077 0.0534 0.1373>, rot=<0.0082 0.0188 0.1708 -0.9851>}
 pinky_metacarpal_l {parent=hand_l, loc=<1.7616 -0.9753 3.3814>, rot=<-0.0397 0.0814 -0.069 -0.9935>}
 pinky_01_l {parent=pinky_metacarpal_l, loc=<7.1709 -0.4831 -0.1094>, rot=<-0.0654 -0.0035 0.2242 -0.9723>}
 pinky_02_l {parent=pinky_01_l, loc=<2.8274 0.0893 -0.1043>, rot=<0.0037 0.0011 0.2854 -0.9584>}
 pinky_03_l {parent=pinky_02_l, loc=<2.2046 0.0307 -0.0078>, rot=<-0.0002 0.0098 0.1445 -0.9895>}
 ring_metacarpal_l {parent=hand_l, loc=<2.2074 -0.3625 1.3638>, rot=<-0.0284 0.0722 -0.0458 -0.9959>}
 ring_01_l {parent=ring_metacarpal_l, loc=<7.0295 -0.6197 -0.1446>, rot=<-0.0489 -0.0252 0.1892 -0.9804>}
 ring_02_l {parent=ring_01_l, loc=<3.6174 0.4647 -0.1027>, rot=<0 0 0.2806 -0.9598>}
 ring_03_l {parent=ring_02_l, loc=<2.4152 0.1141 0.0199>, rot=<0 0 0.1189 -0.9929>}
 thumb_01_l {parent=hand_l, loc=<1.3118 -0.3191 -2.459>, rot=<0.6546 -0.4042 0.0664 -0.6354>}
 thumb_02_l {parent=thumb_01_l, loc=<5.9336 -0.0448 0.0401>, rot=<-0.0037 0.0213 0.298 -0.9543>}
 thumb_03_l {parent=thumb_02_l, loc=<2.2223 0.3682 -0.0907>, rot=<0.0118 -0.013 0.0066 -0.9998>}
 deform_wrist_dn_base_l {parent=hand_l, loc=<0 -2.2342 0>, rot=<0 0 0 -1>}
 deform_wrist_dn_l {parent=deform_wrist_dn_base_l, loc=<0 0 0>, rot=<0 0 0 1>}
 deform_wrist_up_base_l {parent=hand_l, loc=<0 2.2682 0>, rot=<0 0 0 -1>}
 deform_wrist_up_l {parent=deform_wrist_up_base_l, loc=<0 0 0>, rot=<0 0 0 1>}
 dyn_radiusEnd_l {parent=hand_l, loc=<0.4715 0.7461 -0.8402>, rot=<0.5857 -0.566 -0.4799 0.326>}
 dyn_ulnaEnd_l {parent=hand_l, loc=<0.2852 -0.306 1.7325>, rot=<0.5857 -0.566 -0.4799 0.326>}
 lowerarm_twist_01_l {parent=lowerarm_l, loc=<17.8549 0 0>, rot=<0 0 0 1>}
 lowerarm_twist_02_l {parent=lowerarm_l, loc=<9.6199 0 0>, rot=<0 0 0 1>}
 dyn_ladius_l {parent=lowerarm_l, loc=<0.3118 -0.4064 1.6266>, rot=<0.0291 -0.0209 -0.027 -0.999>}
 dyn_ulna_l {parent=lowerarm_l, loc=<-1.2916 -0.5525 -2.3127>, rot=<0.1626 0.0361 0.0312 -0.9855>}
 deform_elbow_in_base_l {parent=lowerarm_l, loc=<0.2112 5.259 0.8274>, rot=<0 0 0 -1>}
 deform_elbow_in_l {parent=deform_elbow_in_base_l, loc=<0 0 0>, rot=<0 0 0 1>}
 deform_elbow_out_base_l {parent=lowerarm_l, loc=<0 -4 0>, rot=<0 0 0 -1>}
 deform_elbow_out_l {parent=deform_elbow_out_base_l, loc=<0 0 0>, rot=<0 0 0 1>}
 dyn_radius_l {parent=lowerarm_l, loc=<0.3118 -0.4064 1.6266>, rot=<0.0291 -0.0209 -0.027 -0.999>}
 upperarm_twist_01_l {parent=upperarm_l, loc=<0.0041 0 0>, rot=<0 0 0 1>}
 deform_upperarm_bk_base_l {parent=upperarm_twist_01_l, loc=<1.0813 -7.2193 -1.9453>, rot=<0 0 0 1>}
 deform_upperarm_bk_l {parent=deform_upperarm_bk_base_l, loc=<0 0 0>, rot=<0 0 0 1>}
 deform_upperarm_fr_base_l {parent=upperarm_twist_01_l, loc=<-0.3691 7.0506 -2.765>, rot=<0 0 0 1>}
 deform_upperarm_fr_l {parent=deform_upperarm_fr_base_l, loc=<0 0 0>, rot=<0 0 0 1>}
 deform_upperarm_up_base_l {parent=upperarm_twist_01_l, loc=<-3.2216 -1.1391 3.3311>, rot=<0 0 0 1>}
 deform_upperarm_up_l {parent=deform_upperarm_up_base_l, loc=<0 0 0>, rot=<0 0 0 1>}
 upperarm_twist_02_l {parent=upperarm_l, loc=<13.7941 0 0>, rot=<0 0 0 1>}
 dyn_clavEnd_Base_l {parent=clavicle_l, loc=<15.1833 1.7381 3.8566>, rot=<0.9398 -0.3224 0.1122 0.0167>}
 dyn_clavEnd_l {parent=dyn_clavEnd_Base_l, loc=<0 0 0>, rot=<0 0 0 -1>}
 dyn_scapEnd_Base_l {parent=clavicle_l, loc=<14.316 -1.6474 3.0125>, rot=<0.9183 0.3934 0.0423 0.0083>}
 dyn_scapEnd_l {parent=dyn_scapEnd_Base_l, loc=<0 0 0>, rot=<0 0 0 -1>}
 clavicle_r {parent=spine_05, loc=<4.3575 -0.1466 2.3223>, rot=<0.7289 0.0465 -0.6829 0.009>}
 upperarm_r {parent=clavicle_r, loc=<-15.5111 0 0>, rot=<0.0226 -0.3234 -0.0088 -0.9459>}
 lowerarm_r {parent=upperarm_r, loc=<-26.6514 0 0>, rot=<0.0391 -0.0099 -0.2453 -0.9686>}
 hand_r {parent=lowerarm_r, loc=<-22.3491 0 0>, rot=<-0.7081 -0.0507 0.0975 -0.6975>}
 index_metacarpal_r {parent=hand_r, loc=<-1.6535 -0.0183 2.96>, rot=<0.0158 -0.0124 -0.0173 -0.9996>}
 index_01_r {parent=index_metacarpal_r, loc=<-8.3008 -0.0357 0.0524>, rot=<-0.0044 0.0485 0.0924 -0.9945>}
 index_02_r {parent=index_01_r, loc=<-3.3587 0.3004 -0.0452>, rot=<-0.002 0.0027 0.3225 -0.9466>}
 index_03_r {parent=index_02_r, loc=<-2.5578 0.1231 0.0251>, rot=<-0.002 0.0067 0.2558 -0.9667>}
 middle_metacarpal_r {parent=hand_r, loc=<-2.1856 0.1725 0.5984>, rot=<0.006 0.0168 -0.0329 -0.9993>}
 middle_01_r {parent=middle_metacarpal_r, loc=<-7.3275 0.2447 -0.0129>, rot=<-0.0208 0.0042 0.0858 -0.9961>}
 middle_02_r {parent=middle_01_r, loc=<-4.0123 -0.1401 -0.1175>, rot=<-0.0218 0.0258 0.3512 -0.9357>}
 middle_03_r {parent=middle_02_r, loc=<-2.6077 -0.0534 -0.1373>, rot=<0.0082 0.0188 0.1708 -0.9851>}
 pinky_metacarpal_r {parent=hand_r, loc=<-1.7616 0.9753 -3.3814>, rot=<-0.0397 0.0814 -0.069 -0.9935>}
 pinky_01_r {parent=pinky_metacarpal_r, loc=<-7.1709 0.4831 0.1094>, rot=<-0.0654 -0.0035 0.2242 -0.9723>}
 pinky_02_r {parent=pinky_01_r, loc=<-2.8274 -0.0892 0.1043>, rot=<0.0037 0.0011 0.2854 -0.9584>}
 pinky_03_r {parent=pinky_02_r, loc=<-2.2046 -0.0307 0.0078>, rot=<-0.0002 0.0098 0.1445 -0.9895>}
 ring_metacarpal_r {parent=hand_r, loc=<-2.2074 0.3625 -1.3638>, rot=<-0.0284 0.0722 -0.0458 -0.9959>}
 ring_01_r {parent=ring_metacarpal_r, loc=<-7.0295 0.6197 0.1446>, rot=<-0.0489 -0.0252 0.1892 -0.9804>}
 ring_02_r {parent=ring_01_r, loc=<-3.6174 -0.4647 0.1027>, rot=<0 0 0.2806 -0.9598>}
 ring_03_r {parent=ring_02_r, loc=<-2.4153 -0.1141 -0.0199>, rot=<0 0 0.1189 -0.9929>}
 thumb_01_r {parent=hand_r, loc=<-1.3118 0.3191 2.459>, rot=<0.6546 -0.4042 0.0664 -0.6354>}
 thumb_02_r {parent=thumb_01_r, loc=<-5.9335 0.0447 -0.0401>, rot=<-0.0037 0.0213 0.298 -0.9543>}
 thumb_03_r {parent=thumb_02_r, loc=<-2.2223 -0.3682 0.0907>, rot=<0.0118 -0.013 0.0066 -0.9998>}
 dyn_radiusEnd_r {parent=hand_r, loc=<-0.4712 -0.7464 0.8402>, rot=<-0.4799 -0.326 -0.5857 -0.566>}
 dyn_ulnaEnd_r {parent=hand_r, loc=<-0.2849 0.3057 -1.7324>, rot=<-0.4799 -0.326 -0.5857 -0.566>}
 deform_wrist_dn_base_r {parent=hand_r, loc=<-0.0004 2.2345 -0.0001>, rot=<0 0 0 -1>}
 deform_wrist_dn_r {parent=deform_wrist_dn_base_r, loc=<0 0 0>, rot=<0 0 0 1>}
 deform_wrist_up_base_r {parent=hand_r, loc=<0.0003 -2.2684 0.0001>, rot=<0 0 0 -1>}
 deform_wrist_up_r {parent=deform_wrist_up_base_r, loc=<0 0 0>, rot=<0 0 0 1>}
 lowerarm_twist_01_r {parent=lowerarm_r, loc=<-17.8549 0 0>, rot=<0 0 0 1>}
 lowerarm_twist_02_r {parent=lowerarm_r, loc=<-9.6199 0 0>, rot=<0 0 0 1>}
 dyn_ulna_r {parent=lowerarm_r, loc=<1.2914 0.5525 2.313>, rot=<-0.0312 -0.9855 0.1626 -0.0361>}
 dyn_radius_r {parent=lowerarm_r, loc=<-0.3119 0.4064 -1.6265>, rot=<0.027 -0.999 0.0291 0.0209>}
 deform_elbow_in_base_r {parent=lowerarm_r, loc=<-0.2111 -5.2591 -0.8275>, rot=<0 0 0 -1>}
 deform_elbow_in_r {parent=deform_elbow_in_base_r, loc=<-0.0001 0 0.0002>, rot=<0 0 0 1>}
 deform_elbow_out_base_r {parent=lowerarm_r, loc=<0 4 -0.0001>, rot=<0 0 0 -1>}
 deform_elbow_out_r {parent=deform_elbow_out_base_r, loc=<0 0 0>, rot=<0 0 0 1>}
 upperarm_twist_01_r {parent=upperarm_r, loc=<-0.0041 0 0>, rot=<0 0 0 1>}
 deform_upperarm_bk_base_r {parent=upperarm_twist_01_r, loc=<-1.0814 7.2192 1.9455>, rot=<0 0 0 1>}
 deform_upperarm_bk_r {parent=deform_upperarm_bk_base_r, loc=<0 0 0>, rot=<0 0 0 1>}
 deform_upperarm_fr_base_r {parent=upperarm_twist_01_r, loc=<0.3689 -7.0506 2.7652>, rot=<0 0 0 1>}
 deform_upperarm_fr_r {parent=deform_upperarm_fr_base_r, loc=<0 0 0>, rot=<0 0 0 1>}
 deform_upperarm_up_base_r {parent=upperarm_twist_01_r, loc=<3.2214 1.1391 -3.3308>, rot=<0 0 0 1>}
 deform_upperarm_up_r {parent=deform_upperarm_up_base_r, loc=<0 0 0>, rot=<0 0 0 -1>}
 upperarm_twist_02_r {parent=upperarm_r, loc=<-13.7942 0 0>, rot=<0 0 0 1>}
 dyn_scapEnd_Base_r {parent=clavicle_r, loc=<-14.316 1.6474 -3.0121>, rot=<-0.0423 0.0083 0.9183 -0.3934>}
 dyn_scapEnd_r {parent=dyn_scapEnd_Base_r, loc=<0 0 0>, rot=<0 0 0 1>}
 dyn_clavEnd_Base_r {parent=clavicle_r, loc=<-15.1833 -1.7381 -3.8561>, rot=<-0.1122 0.0167 0.9398 0.3224>}
 dyn_clavEnd_r {parent=dyn_clavEnd_Base_r, loc=<0 0 0>, rot=<0 0 0 1>}
 neck_01 {parent=spine_05, loc=<11.3138 0.1529 0>, rot=<0 0 -0.2003 -0.9797>}
 neck_02 {parent=neck_01, loc=<4.5627 0 0>, rot=<0 0 0.0504 -0.9987>}
 head {parent=neck_02, loc=<5.3914 0 0>, rot=<0 0 0.115 -0.9934>}
 attach_cape {parent=spine_05, loc=<9.017 -11.714 0>, rot=<0.0616 -0.7044 -0.0616 -0.7044>}
 cape {parent=attach_cape, type=Null, loc=<0 0 0>, rot=<0 0 0 1>}
 attach_backpack {parent=spine_05, loc=<-0.123 -13.473 0>, rot=<-0.0093 -0.707 0.0093 -0.707>}
 Backpack_BR {parent=attach_backpack, type=Null, loc=<0 -6.6861 0>, rot=<0 0 0 1>}
 dyn_neckRibA_r {parent=spine_05, loc=<18.7376 -10.918 2.6195>, rot=<-0.0815 -0.0322 -0.844 -0.5292>}
 dyn_neckRibB_r {parent=spine_05, loc=<12.8582 -11.8247 2.9145>, rot=<-0.0643 -0.0578 -0.814 -0.5745>}
 dyn_neckRibA_l {parent=spine_05, loc=<18.7372 -10.918 -2.6195>, rot=<0.844 -0.5292 0.0815 -0.0322>}
 dyn_neckRibB_l {parent=spine_05, loc=<12.8579 -11.8247 -2.9145>, rot=<0.814 -0.5745 0.0643 -0.0578>}
 dyn_clav_Base_r {parent=spine_05, loc=<2.3385 14.536 1.6274>, rot=<-0.5933 -0.2891 0.7257 0.1944>}
 dyn_clav_r {parent=dyn_clav_Base_r, loc=<0 0 0>, rot=<0 0 0 1>}
 dyn_scap_Base_r {parent=spine_05, loc=<7.1412 -14.9934 3.36>, rot=<0.6587 -0.2388 -0.6541 0.285>}
 dyn_scap_r {parent=dyn_scap_Base_r, loc=<0 0 0>, rot=<0 0 0 1>}
 dyn_clav_Base_l {parent=spine_05, loc=<2.3387 14.536 -1.6274>, rot=<0.7257 -0.1944 -0.5933 0.2891>}
 dyn_clav_l {parent=dyn_clav_Base_l, loc=<0 0 0>, rot=<0 0 0 1>}
 dyn_scap_Base_l {parent=spine_05, loc=<7.1415 -14.9934 -3.36>, rot=<-0.6541 -0.285 0.6587 0.2388>}
 dyn_scap_l {parent=dyn_scap_Base_l, loc=<0 0 0>, rot=<0 0 0 1>}
 FX_Fire_0 {parent=spine_05, type=Null, loc=<0 0 0>, rot=<0 0 0 1>}
 FX_Fire_1 {parent=spine_05, type=Null, loc=<0 -16.8772 -12.7868>, rot=<0 0.7071 0 0.7071>}
 dyn_ribA_Base_r {parent=spine_04, loc=<12.7062 0.364 3.4276>, rot=<0.3743 -0.5289 0.6562 0.3868>}
 dyn_ribA_r {parent=dyn_ribA_Base_r, loc=<0 0 0>, rot=<0 0 0 1>}
 dyn_ribB_Base_r {parent=spine_04, loc=<7.8525 0.1887 2.0666>, rot=<0.3844 -0.5319 0.6462 0.3896>}
 dyn_ribB_r {parent=dyn_ribB_Base_r, loc=<0 0 0>, rot=<0 0 0 1>}
 dyn_ribC_Base_r {parent=spine_04, loc=<1.984 -0.4674 1.7195>, rot=<0.4015 -0.5202 0.6329 0.4094>}
 dyn_ribC_r {parent=dyn_ribC_Base_r, loc=<0 0 0>, rot=<0 0 0 1>}
 dyn_ribD_Base_r {parent=spine_04, loc=<-2.8102 -0.6033 0.4655>, rot=<0.4148 -0.5599 0.6099 0.3776>}
 dyn_ribD_r {parent=dyn_ribD_Base_r, loc=<0 0 0>, rot=<0 0 0 1>}
 dyn_libA_Base_l {parent=spine_04, loc=<12.7067 0.364 -3.4276>, rot=<-0.3868 -0.6562 0.5289 -0.3743>}
 dyn_libA_l {parent=dyn_libA_Base_l, loc=<0 0 0>, rot=<0 0 0 1>}
 dyn_libB_Base_l {parent=spine_04, loc=<7.8525 0.1887 -2.0666>, rot=<-0.3896 -0.6462 0.5319 -0.3844>}
 dyn_libB_l {parent=dyn_libB_Base_l, loc=<0 0 0>, rot=<0 0 0 1>}
 dyn_libC_Base_l {parent=spine_04, loc=<1.9837 -0.4674 -1.7195>, rot=<-0.4094 -0.6329 0.5202 -0.4015>}
 dyn_libC_l {parent=dyn_libC_Base_l, loc=<0 0 0>, rot=<0 0 0 1>}
 dyn_libD_Base_l {parent=spine_04, loc=<-2.8106 -0.6034 -0.4655>, rot=<-0.3776 -0.6099 0.5599 -0.4148>}
 dyn_libD_l {parent=dyn_libD_Base_l, loc=<0 0 0>, rot=<0 0 0 1>}
 deform_pec_aim_base_l {parent=spine_04, loc=<7.8732 12.6388 -2.4123>, rot=<0.0102 -0.4602 0.3979 -0.7936>}
 deform_pec_aim_l {parent=deform_pec_aim_base_l, loc=<0 0 0>, rot=<0 0 0 1>}
 deform_pec_base_l {parent=deform_pec_aim_l, loc=<10.766 0 0>, rot=<0 0 0 1>}
 deform_pec_l {parent=deform_pec_base_l, loc=<0.055 0 0>, rot=<0 0 0 1>}
 deform_pec_aim_base_r {parent=spine_04, loc=<7.8727 12.6388 2.4123>, rot=<-0.4602 -0.0102 0.7936 0.3979>}
 deform_pec_aim_r {parent=deform_pec_aim_base_r, loc=<0 0 0>, rot=<0 0 0 1>}
 deform_pec_base_r {parent=deform_pec_aim_r, loc=<-10.7662 -0.0006 -0.0007>, rot=<0 0 0 1>}
 deform_pec_r {parent=deform_pec_base_r, loc=<0 0 0>, rot=<0 0 0 1>}
 dyn_ribA_Base_l {parent=spine_04, loc=<12.7067 0.364 -3.4276>, rot=<-0.3868 -0.6562 0.5289 -0.3743>}
 dyn_ribA_l {parent=dyn_ribA_Base_l, loc=<0 0 0>, rot=<0 0 0 1>}
 dyn_ribB_Base_l {parent=spine_04, loc=<7.8525 0.1887 -2.0666>, rot=<-0.3896 -0.6462 0.5319 -0.3844>}
 dyn_ribB_l {parent=dyn_ribB_Base_l, loc=<0 0 0>, rot=<0 0 0 1>}
 dyn_ribC_Base_l {parent=spine_04, loc=<1.9837 -0.4674 -1.7195>, rot=<-0.4094 -0.6329 0.5202 -0.4015>}
 dyn_ribC_l {parent=dyn_ribC_Base_l, loc=<0 0 0>, rot=<0 0 0 1>}
 dyn_ribD_Base_l {parent=spine_04, loc=<-2.8106 -0.6034 -0.4655>, rot=<-0.3776 -0.6099 0.5599 -0.4148>}
 dyn_ribD_l {parent=dyn_ribD_Base_l, loc=<0 0 0>, rot=<0 0 0 1>}
 thigh_l {parent=pelvis, loc=<-2.5693 -0.0109 -7.7501>, rot=<0.0012 0.0572 0.0155 -0.9982>}
 calf_l {parent=thigh_l, loc=<-39.0484 0 0>, rot=<0.0006 -0.0161 -0.0391 -0.9991>}
 calf_twist_01_l {parent=calf_l, loc=<-35.8131 -0.0967 0.0397>, rot=<0.0237 0.0095 -0.0035 -0.9997>}
 calf_twist_02_l {parent=calf_l, loc=<-18.8064 -0.0967 0.0397>, rot=<0.0237 0.0095 -0.0035 -0.9997>}
 foot_l {parent=calf_l, loc=<-42.561 0.1077 0.4292>, rot=<0.0418 -0.0455 0.0343 -0.9975>}
 ball_l {parent=foot_l, loc=<-6.1599 13.8264 0.6806>, rot=<0.0001 -0.0001 -0.7063 -0.7079>}
 dyn_ball_l {parent=foot_l, loc=<-7.3637 9.0115 0.6814>, rot=<0.7063 -0.7079 0.0001 0.0001>}
 dyn_fibulaEnd_l {parent=foot_l, loc=<-0.4035 -0.9053 -2.2626>, rot=<0.7045 -0.0357 0.7082 -0.0284>}
 dyn_tibiaEnd_l {parent=foot_l, loc=<0.0976 -0.6413 1.22>, rot=<0.7045 -0.0357 0.7082 -0.0284>}
 dyn_tibia_Base_l {parent=calf_l, loc=<-1.8244 -2.2745 0.9984>, rot=<-0.023 -0.9986 -0.046 0.0074>}
 dyn_tibia_l {parent=dyn_tibia_Base_l, loc=<0 0 0>, rot=<0 0 0 1>}
 dyn_fibula_Base_l {parent=calf_l, loc=<-1.1732 -2.1633 -2.1061>, rot=<-0.0144 -0.9998 0.0118 0.005>}
 dyn_fibula_l {parent=dyn_fibula_Base_l, loc=<0 0 0>, rot=<0 0 0 1>}
 dyn_kneeToHip_l {parent=calf_l, loc=<-0.0001 0 -0.0001>, rot=<-0.0013 0.0027 0.0462 -0.9989>}
 deform_knee_in_base_l {parent=calf_l, loc=<1.298 -5.2957 -0.0496>, rot=<0 0 0 -1>}
 deform_knee_in_l {parent=deform_knee_in_base_l, loc=<0 0 0>, rot=<0 0 0 1>}
 deform_knee_out_base_l {parent=calf_l, loc=<0 4 0>, rot=<0 0 0 -1>}
 deform_knee_out_l {parent=deform_knee_out_base_l, loc=<0 0 0>, rot=<0 0 0 1>}
 thigh_twist_01_l {parent=thigh_l, loc=<-21.6303 0 0>, rot=<0.0013 0 0 -1>}
 thigh_twist_shift_l {parent=thigh_twist_01_l, loc=<0 0 0>, rot=<0 0 0 1>}
 deform_groin_base_l {parent=thigh_l, loc=<-2.1908 8.1142 -1.7442>, rot=<0 0 0 -1>}
 deform_groin_l {parent=deform_groin_base_l, loc=<0 0 0>, rot=<0 0 0 1>}
 thigh_r {parent=pelvis, loc=<-2.5693 -0.0109 7.7501>, rot=<0.0572 -0.0012 0.9982 0.0155>}
 calf_r {parent=thigh_r, loc=<39.0484 0 0>, rot=<0.0006 -0.0161 -0.0391 -0.9991>}
 calf_twist_01_r {parent=calf_r, loc=<35.813 0.0967 -0.0397>, rot=<0.0237 0.0095 -0.0035 -0.9997>}
 calf_twist_02_r {parent=calf_r, loc=<18.8064 0.0967 -0.0397>, rot=<0.0237 0.0095 -0.0035 -0.9997>}
 foot_r {parent=calf_r, loc=<42.561 -0.1077 -0.4292>, rot=<0.0418 -0.0455 0.0343 -0.9975>}
 ball_r {parent=foot_r, loc=<6.1598 -13.8264 -0.6806>, rot=<0.0001 -0.0001 -0.7063 -0.7079>}
 dyn_ball_r {parent=foot_r, loc=<7.3637 -9.0115 -0.6814>, rot=<0.0001 -0.0001 -0.7063 -0.7079>}
 dyn_fibulaEnd_r {parent=foot_r, loc=<0.4035 0.9053 2.2626>, rot=<0.7082 0.0284 -0.7045 -0.0357>}
 dyn_tibiaEnd_r {parent=foot_r, loc=<-0.0976 0.6413 -1.22>, rot=<0.7082 0.0284 -0.7045 -0.0357>}
 dyn_kneeToHip_r {parent=calf_r, loc=<0 0 0>, rot=<-0.0027 -0.0013 0.9989 0.0462>}
 dyn_tibia_Base_r {parent=calf_r, loc=<1.8244 2.2746 -0.9984>, rot=<-0.046 -0.0074 0.023 -0.9986>}
 dyn_tibia_r {parent=dyn_tibia_Base_r, loc=<0 0 0>, rot=<0 0 0 1>}
 dyn_fibula_Base_r {parent=calf_r, loc=<1.1733 2.1633 2.1062>, rot=<0.0118 -0.005 0.0144 -0.9998>}
 dyn_fibula_r {parent=dyn_fibula_Base_r, loc=<0 0 0>, rot=<0 0 0 1>}
 deform_knee_in_base_r {parent=calf_r, loc=<-1.298 5.2957 0.0496>, rot=<0 0 0 -1>}
 deform_knee_in_r {parent=deform_knee_in_base_r, loc=<0 0 0>, rot=<0 0 0 1>}
 deform_knee_out_base_r {parent=calf_r, loc=<0 -4 0>, rot=<0 0 0 -1>}
 deform_knee_out_r {parent=deform_knee_out_base_r, loc=<0 0 0>, rot=<0 0 0 1>}
 thigh_twist_01_r {parent=thigh_r, loc=<21.6303 0 0>, rot=<0.0013 0 0 -1>}
 thigh_twist_shift_r {parent=thigh_twist_01_r, loc=<-0.0002 0 0>, rot=<0 0 0 1>}
 deform_groin_base_r {parent=thigh_r, loc=<2.1908 -8.1142 1.7442>, rot=<0 0 0 -1>}
 deform_groin_r {parent=deform_groin_base_r, loc=<0 0 0>, rot=<0 0 0 1>}
 dyn_hipToKnee_r {parent=pelvis, loc=<-9.692 -0.2474 9.4244>, rot=<0.0458 0.059 0.997 0.0205>}
 dyn_hipToKnee_l {parent=pelvis, loc=<-9.692 -0.2474 -9.4244>, rot=<-0.0458 -0.059 0.997 0.0205>}
 dyn_pelvis_r {parent=pelvis, loc=<-9.692 -0.2474 9.4244>, rot=<0.0458 0.059 0.997 0.0205>}
 dyn_pelvis_l {parent=pelvis, loc=<-9.692 -0.2474 -9.4244>, rot=<-0.059 0.0458 0.0205 -0.997>}
 deform_glute_base_l {parent=pelvis, loc=<-8.8064 -8.8222 -8.4287>, rot=<0.0275 0.0501 -0.4472 -0.8926>}
 deform_glute_l {parent=deform_glute_base_l, loc=<0 0 0>, rot=<0 0 0 1>}
 deform_glute_base_r {parent=pelvis, loc=<-8.8062 -8.8222 8.4287>, rot=<0.0501 -0.0275 0.8926 -0.4472>}
 deform_glute_r {parent=deform_glute_base_r, loc=<0 0 0>, rot=<0 0 0 1>}
]
