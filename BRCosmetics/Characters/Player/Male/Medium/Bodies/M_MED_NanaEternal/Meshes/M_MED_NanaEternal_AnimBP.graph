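# Animation Blueprint for the NanaEternal male medium body cosmetic that adjusts its dynamic pelvis, clavicle and rib bones during targeting and emotes.
# It copies the attached parent mesh's pose into a Control Rig, then applies an always-on additive pelvis offset.
# Two identical Modify Bone chains shrink and nudge neck, clavicle and rib bones, driven by Target and Emote alphas.
# Events set these alphas to 0 or 1 with interpolated blending, and both chains stack when active together.
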AnimBlueprint /BRCosmetics/Characters/Player/Male/Medium/Bodies/M_MED_NanaEternal/Meshes/M_MED_NanaEternal_AnimBP
target_skeleton=&/M_MED_NanaEternal_Skeleton
graphs [
 AnimGraph {nodes [
   n1 = Control Rig(Source=Copy Pose From Mesh() {bUseAttachedParent=true}) {
    alphaBoolBlend {blendOption=Linear}
    controlRigAssetReference {blueprintRigClass=&/M_MED_NanaEternal_CtrlRig.M_MED_NanaEternal_CtrlRig_C}
   }
   n2 = "Transform (Modify) Bone - Bone: dyn_pelvis_r"(ComponentPose=Local To Component(LocalPose=@n1), Alpha=1, Translation="1,0,2", Rotation="0,0,0", Scale="0.9,0.9,0.9") {
    rotationMode=BMM_Additive
    scaleMode=BMM_Additive
    translationMode=BMM_Additive
    alphaBoolBlend {blendOption=Linear}
    boneToModify {boneName=dyn_pelvis_r}
    scale=<0.9 0.9 0.9>
    translation {x=1, z=2}
   }
   n3 = "Transform (Modify) Bone - Bone: dyn_pelvis_l"(ComponentPose=@n2, Alpha=1, Translation="-1,0,2", Rotation="0,0,0", Scale="0.9,0.9,0.9") {
    rotationMode=BMM_Additive
    scaleMode=BMM_Additive
    translationMode=BMM_Additive
    alphaBoolBlend {blendOption=Linear}
    boneToModify {boneName=dyn_pelvis_l}
    scale=<0.9 0.9 0.9>
    translation {x=-1, z=2}
   }
   n4 = Get Target() -> (Target:double)
   n5 = "Transform (Modify) Bone - Bone: dyn_neckRibA_l"(ComponentPose=@n3, Alpha=@n4, Translation="0,0,0", Rotation="0,0,0", Scale="0.9,1,1") {
    rotationMode=BMM_Additive
    scaleMode=BMM_Additive
    translationMode=BMM_Additive
    alphaBoolBlend {blendOption=Linear}
    alphaScaleBiasClamp {bInterpResult=true, interpSpeedDecreasing=5, interpSpeedIncreasing=5}
    boneToModify {boneName=dyn_neckRibA_l}
    scale {x=0.9}
   }
   n6 = "Transform (Modify) Bone - Bone: dyn_neckRibA_r"(ComponentPose=@n5, Alpha=@n4, Translation="0,0,0", Rotation="0,0,0", Scale="0.9,1,1") {
    rotationMode=BMM_Additive
    scaleMode=BMM_Additive
    translationMode=BMM_Additive
    alphaBoolBlend {blendOption=Linear}
    alphaScaleBiasClamp {bInterpResult=true, interpSpeedDecreasing=5, interpSpeedIncreasing=5}
    boneToModify {boneName=dyn_neckRibA_r}
    scale {x=0.9}
   }
   n7 = "Transform (Modify) Bone - Bone: dyn_neckRibB_l"(ComponentPose=@n6, Alpha=@n4, Translation="0,0,0", Rotation="0,0,0", Scale="0.9,1,1") {
    rotationMode=BMM_Additive
    scaleMode=BMM_Additive
    translationMode=BMM_Additive
    alphaBoolBlend {blendOption=Linear}
    alphaScaleBiasClamp {bInterpResult=true, interpSpeedDecreasing=5, interpSpeedIncreasing=5}
    boneToModify {boneName=dyn_neckRibB_l}
    scale {x=0.9}
   }
   n8 = "Transform (Modify) Bone - Bone: dyn_neckRibB_r"(ComponentPose=@n7, Alpha=@n4, Translation="0,0,0", Rotation="0,0,0", Scale="0.9,1,1") {
    rotationMode=BMM_Additive
    scaleMode=BMM_Additive
    translationMode=BMM_Additive
    alphaBoolBlend {blendOption=Linear}
    alphaScaleBiasClamp {bInterpResult=true, interpSpeedDecreasing=5, interpSpeedIncreasing=5}
    boneToModify {boneName=dyn_neckRibB_r}
    scale {x=0.9}
   }
   n9 = "Transform (Modify) Bone - Bone: dyn_clav_l"(ComponentPose=@n8, Alpha=@n4, Translation="-0.5,0,0", Rotation="0,-7,0", Scale="0.9,0.9,0.9") {
    rotationMode=BMM_Additive
    scaleMode=BMM_Additive
    translationMode=BMM_Additive
    alphaBoolBlend {blendOption=Linear}
    alphaScaleBiasClamp {bInterpResult=true, interpSpeedDecreasing=5, interpSpeedIncreasing=5}
    boneToModify {boneName=dyn_clav_l}
    rotation {yaw=-7}
    scale=<0.9 0.9 0.9>
    translation {x=-0.5}
   }
   n10 = "Transform (Modify) Bone - Bone: dyn_clav_r"(ComponentPose=@n9, Alpha=@n4, Translation="0.5,0,0", Rotation="0,5,0", Scale="0.9,0.9,0.9") {
    rotationMode=BMM_Additive
    scaleMode=BMM_Additive
    translationMode=BMM_Additive
    alphaBoolBlend {blendOption=Linear}
    alphaScaleBiasClamp {bInterpResult=true, interpSpeedDecreasing=5, interpSpeedIncreasing=5}
    boneToModify {boneName=dyn_clav_r}
    rotation {yaw=5}
    scale=<0.9 0.9 0.9>
    translation {x=0.5}
   }
   n11 = "Transform (Modify) Bone - Bone: dyn_ribA_r"(ComponentPose=@n10, Alpha=@n4, Translation="1,0,0", Rotation="0,0,0", Scale="0.9,1,1") {
    rotationMode=BMM_Additive
    scaleMode=BMM_Additive
    translationMode=BMM_Additive
    alphaBoolBlend {blendOption=Linear}
    alphaScaleBiasClamp {bInterpResult=true, interpSpeedDecreasing=5, interpSpeedIncreasing=5}
    boneToModify {boneName=dyn_ribA_r}
    scale {x=0.9}
    translation {x=1}
   }
   n12 = "Transform (Modify) Bone - Bone: dyn_ribA_l"(ComponentPose=@n11, Alpha=@n4, Translation="-1,0,0", Rotation="0,0,0", Scale="0.9,1,1") {
    rotationMode=BMM_Additive
    scaleMode=BMM_Additive
    translationMode=BMM_Additive
    alphaBoolBlend {blendOption=Linear}
    alphaScaleBiasClamp {bInterpResult=true, interpSpeedDecreasing=5, interpSpeedIncreasing=5}
    boneToModify {boneName=dyn_ribA_l}
    scale {x=0.9}
    translation {x=-1}
   }
   n13 = "Transform (Modify) Bone - Bone: dyn_ribB_l"(ComponentPose=@n12, Alpha=@n4, Translation="0,0,0", Rotation="0,0,0", Scale="0.9,1,1") {
    rotationMode=BMM_Additive
    scaleMode=BMM_Additive
    translationMode=BMM_Additive
    alphaBoolBlend {blendOption=Linear}
    alphaScaleBiasClamp {bInterpResult=true, interpSpeedDecreasing=5, interpSpeedIncreasing=5}
    boneToModify {boneName=dyn_ribB_l}
    scale {x=0.9}
   }
   n14 = "Transform (Modify) Bone - Bone: dyn_ribB_r"(ComponentPose=@n13, Alpha=@n4, Translation="0,0,0", Rotation="0,0,0", Scale="0.9,1,1") {
    rotationMode=BMM_Additive
    scaleMode=BMM_Additive
    translationMode=BMM_Additive
    alphaBoolBlend {blendOption=Linear}
    alphaScaleBiasClamp {bInterpResult=true, interpSpeedDecreasing=5, interpSpeedIncreasing=5}
    boneToModify {boneName=dyn_ribB_r}
    scale {x=0.9}
   }
   n15 = "Transform (Modify) Bone - Bone: dyn_ribC_l"(ComponentPose=@n14, Alpha=@n4, Translation="0,0,0", Rotation="0,0,0", Scale="0.9,1,1") {
    rotationMode=BMM_Additive
    scaleMode=BMM_Additive
    translationMode=BMM_Additive
    alphaBoolBlend {blendOption=Linear}
    alphaScaleBiasClamp {bInterpResult=true, interpSpeedDecreasing=5, interpSpeedIncreasing=5}
    boneToModify {boneName=dyn_ribC_l}
    scale {x=0.9}
   }
   n16 = "Transform (Modify) Bone - Bone: dyn_ribC_r"(ComponentPose=@n15, Alpha=@n4, Translation="0,0,0", Rotation="0,0,0", Scale="0.9,1,1") {
    rotationMode=BMM_Additive
    scaleMode=BMM_Additive
    translationMode=BMM_Additive
    alphaBoolBlend {blendOption=Linear}
    alphaScaleBiasClamp {bInterpResult=true, interpSpeedDecreasing=5, interpSpeedIncreasing=5}
    boneToModify {boneName=dyn_ribC_r}
    scale {x=0.9}
   }
   n17 = "Transform (Modify) Bone - Bone: dyn_ribD_l"(ComponentPose=@n16, Alpha=@n4, Translation="0,0,0", Rotation="0,0,0", Scale="0.9,1,1") {
    rotationMode=BMM_Additive
    scaleMode=BMM_Additive
    translationMode=BMM_Additive
    alphaBoolBlend {blendOption=Linear}
    alphaScaleBiasClamp {bInterpResult=true, interpSpeedDecreasing=5, interpSpeedIncreasing=5}
    boneToModify {boneName=dyn_ribD_l}
    scale {x=0.9}
   }
   n18 = "Transform (Modify) Bone - Bone: dyn_ribD_r"(ComponentPose=@n17, Alpha=@n4, Translation="0,0,0", Rotation="0,0,0", Scale="0.9,1,1") {
    rotationMode=BMM_Additive
    scaleMode=BMM_Additive
    translationMode=BMM_Additive
    alphaBoolBlend {blendOption=Linear}
    alphaScaleBiasClamp {bInterpResult=true, interpSpeedDecreasing=5, interpSpeedIncreasing=5}
    boneToModify {boneName=dyn_ribD_r}
    scale {x=0.9}
   }
   n19 = Get Emote() -> (Emote:double)
   n20 = "Transform (Modify) Bone - Bone: dyn_neckRibA_l"(ComponentPose=@n18, Alpha=@n19, Translation="0,0,0", Rotation="0,0,0", Scale="0.9,1,1") {
    rotationMode=BMM_Additive
    scaleMode=BMM_Additive
    translationMode=BMM_Additive
    alphaBoolBlend {blendOption=Linear}
    alphaScaleBiasClamp {bInterpResult=true, interpSpeedDecreasing=5, interpSpeedIncreasing=5}
    boneToModify {boneName=dyn_neckRibA_l}
    scale {x=0.9}
   }
   n21 = "Transform (Modify) Bone - Bone: dyn_neckRibA_r"(ComponentPose=@n20, Alpha=@n19, Translation="0,0,0", Rotation="0,0,0", Scale="0.9,1,1") {
    rotationMode=BMM_Additive
    scaleMode=BMM_Additive
    translationMode=BMM_Additive
    alphaBoolBlend {blendOption=Linear}
    alphaScaleBiasClamp {bInterpResult=true, interpSpeedDecreasing=5, interpSpeedIncreasing=5}
    boneToModify {boneName=dyn_neckRibA_r}
    scale {x=0.9}
   }
   n22 = "Transform (Modify) Bone - Bone: dyn_neckRibB_l"(ComponentPose=@n21, Alpha=@n19, Translation="0,0,0", Rotation="0,0,0", Scale="0.9,1,1") {
    rotationMode=BMM_Additive
    scaleMode=BMM_Additive
    translationMode=BMM_Additive
    alphaBoolBlend {blendOption=Linear}
    alphaScaleBiasClamp {bInterpResult=true, interpSpeedDecreasing=5, interpSpeedIncreasing=5}
    boneToModify {boneName=dyn_neckRibB_l}
    scale {x=0.9}
   }
   n23 = "Transform (Modify) Bone - Bone: dyn_neckRibB_r"(ComponentPose=@n22, Alpha=@n19, Translation="0,0,0", Rotation="0,0,0", Scale="0.9,1,1") {
    rotationMode=BMM_Additive
    scaleMode=BMM_Additive
    translationMode=BMM_Additive
    alphaBoolBlend {blendOption=Linear}
    alphaScaleBiasClamp {bInterpResult=true, interpSpeedDecreasing=5, interpSpeedIncreasing=5}
    boneToModify {boneName=dyn_neckRibB_r}
    scale {x=0.9}
   }
   n24 = "Transform (Modify) Bone - Bone: dyn_clav_l"(ComponentPose=@n23, Alpha=@n19, Translation="-0.5,0,0", Rotation="0,-7,0", Scale="0.9,0.9,0.9") {
    rotationMode=BMM_Additive
    scaleMode=BMM_Additive
    translationMode=BMM_Additive
    alphaBoolBlend {blendOption=Linear}
    alphaScaleBiasClamp {bInterpResult=true, interpSpeedDecreasing=5, interpSpeedIncreasing=5}
    boneToModify {boneName=dyn_clav_l}
    rotation {yaw=-7}
    scale=<0.9 0.9 0.9>
    translation {x=-0.5}
   }
   n25 = "Transform (Modify) Bone - Bone: dyn_clav_r"(ComponentPose=@n24, Alpha=@n19, Translation="0.5,0,0", Rotation="0,5,0", Scale="0.9,0.9,0.9") {
    rotationMode=BMM_Additive
    scaleMode=BMM_Additive
    translationMode=BMM_Additive
    alphaBoolBlend {blendOption=Linear}
    alphaScaleBiasClamp {bInterpResult=true, interpSpeedDecreasing=5, interpSpeedIncreasing=5}
    boneToModify {boneName=dyn_clav_r}
    rotation {yaw=5}
    scale=<0.9 0.9 0.9>
    translation {x=0.5}
   }
   n26 = "Transform (Modify) Bone - Bone: dyn_ribA_r"(ComponentPose=@n25, Alpha=@n19, Translation="1,0,0", Rotation="0,0,0", Scale="0.9,1,1") {
    rotationMode=BMM_Additive
    scaleMode=BMM_Additive
    translationMode=BMM_Additive
    alphaBoolBlend {blendOption=Linear}
    alphaScaleBiasClamp {bInterpResult=true, interpSpeedDecreasing=5, interpSpeedIncreasing=5}
    boneToModify {boneName=dyn_ribA_r}
    scale {x=0.9}
    translation {x=1}
   }
   n27 = "Transform (Modify) Bone - Bone: dyn_ribA_l"(ComponentPose=@n26, Alpha=@n19, Translation="-1,0,0", Rotation="0,0,0", Scale="0.9,1,1") {
    rotationMode=BMM_Additive
    scaleMode=BMM_Additive
    translationMode=BMM_Additive
    alphaBoolBlend {blendOption=Linear}
    alphaScaleBiasClamp {bInterpResult=true, interpSpeedDecreasing=5, interpSpeedIncreasing=5}
    boneToModify {boneName=dyn_ribA_l}
    scale {x=0.9}
    translation {x=-1}
   }
   n28 = "Transform (Modify) Bone - Bone: dyn_ribB_l"(ComponentPose=@n27, Alpha=@n19, Translation="0,0,0", Rotation="0,0,0", Scale="0.9,1,1") {
    rotationMode=BMM_Additive
    scaleMode=BMM_Additive
    translationMode=BMM_Additive
    alphaBoolBlend {blendOption=Linear}
    alphaScaleBiasClamp {bInterpResult=true, interpSpeedDecreasing=5, interpSpeedIncreasing=5}
    boneToModify {boneName=dyn_ribB_l}
    scale {x=0.9}
   }
   n29 = "Transform (Modify) Bone - Bone: dyn_ribB_r"(ComponentPose=@n28, Alpha=@n19, Translation="0,0,0", Rotation="0,0,0", Scale="0.9,1,1") {
    rotationMode=BMM_Additive
    scaleMode=BMM_Additive
    translationMode=BMM_Additive
    alphaBoolBlend {blendOption=Linear}
    alphaScaleBiasClamp {bInterpResult=true, interpSpeedDecreasing=5, interpSpeedIncreasing=5}
    boneToModify {boneName=dyn_ribB_r}
    scale {x=0.9}
   }
   n30 = "Transform (Modify) Bone - Bone: dyn_ribC_l"(ComponentPose=@n29, Alpha=@n19, Translation="0,0,0", Rotation="0,0,0", Scale="0.9,1,1") {
    rotationMode=BMM_Additive
    scaleMode=BMM_Additive
    translationMode=BMM_Additive
    alphaBoolBlend {blendOption=Linear}
    alphaScaleBiasClamp {bInterpResult=true, interpSpeedDecreasing=5, interpSpeedIncreasing=5}
    boneToModify {boneName=dyn_ribC_l}
    scale {x=0.9}
   }
   n31 = "Transform (Modify) Bone - Bone: dyn_ribC_r"(ComponentPose=@n30, Alpha=@n19, Translation="0,0,0", Rotation="0,0,0", Scale="0.9,1,1") {
    rotationMode=BMM_Additive
    scaleMode=BMM_Additive
    translationMode=BMM_Additive
    alphaBoolBlend {blendOption=Linear}
    alphaScaleBiasClamp {bInterpResult=true, interpSpeedDecreasing=5, interpSpeedIncreasing=5}
    boneToModify {boneName=dyn_ribC_r}
    scale {x=0.9}
   }
   n32 = "Transform (Modify) Bone - Bone: dyn_ribD_l"(ComponentPose=@n31, Alpha=@n19, Translation="0,0,0", Rotation="0,0,0", Scale="0.9,1,1") {
    rotationMode=BMM_Additive
    scaleMode=BMM_Additive
    translationMode=BMM_Additive
    alphaBoolBlend {blendOption=Linear}
    alphaScaleBiasClamp {bInterpResult=true, interpSpeedDecreasing=5, interpSpeedIncreasing=5}
    boneToModify {boneName=dyn_ribD_l}
    scale {x=0.9}
   }
   n33 = "Transform (Modify) Bone - Bone: dyn_ribD_r"(ComponentPose=@n32, Alpha=@n19, Translation="0,0,0", Rotation="0,0,0", Scale="0.9,1,1") {
    rotationMode=BMM_Additive
    scaleMode=BMM_Additive
    translationMode=BMM_Additive
    alphaBoolBlend {blendOption=Linear}
    alphaScaleBiasClamp {bInterpResult=true, interpSpeedDecreasing=5, interpSpeedIncreasing=5}
    boneToModify {boneName=dyn_ribD_r}
    scale {x=0.9}
   }
   Output Pose(Result=Component To Local(ComponentPose=@n33)) {layerGroup=None}
  ]}
 EventGraph {nodes [
   Event OnBeginPlayingEmote() {is=title K2Node_Event}
   > Set Emote(Emote:double=1) -> (Output_Get:double)
   Event OnEndPlayingEmote() {is=title K2Node_Event}
   > Set Emote(Emote=0)
   Event OnBeginTargeting() {is=title K2Node_Event}
   > Set Target(Target:double=1) -> (Output_Get:double)
   Event OnEndTargeting() {is=title K2Node_Event}
   > Set Target(Target=0)
   Event BlueprintUpdateAnimation()
  ]}
]
variables [
 Emote {type=number}
 Target {type=number}
]
