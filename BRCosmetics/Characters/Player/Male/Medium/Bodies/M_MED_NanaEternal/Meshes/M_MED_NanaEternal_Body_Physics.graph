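# Physics asset for the NanaEternal medium male player body cosmetic, covering dynamic rib bones on both sides, apparently for secondary motion.
# Four box bodies per side (ribA–D right, libA–D left) each pair with a tiny capsule base body, alongside an unconstrained spine_05 torso capsule.
# Each pair's constraint locks twist and swing2, leaving swing1 as a hinge-like axis with limit angles widening from A to C and D.
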PhysicsAsset /BRCosmetics/Characters/Player/Male/Medium/Bodies/M_MED_NanaEternal/Meshes/M_MED_NanaEternal_Body_Physics
bodies {
 dyn_libA_Base_l [{shapeType=Capsule, length=0.505, radius=0.505, shapeName=dyn_libA_Base_l_capsule, rotation=<rot 0 -90 0>}]
 dyn_libA_l [{shapeType=Box, extentX=5.2282, extentY=18.0942, extentZ=9.535, center=<3.9534 0 1.1635>, rotation=<rot -5.065 -90 -17.9054>}]
 dyn_libB_Base_l [{shapeType=Capsule, length=0.505, radius=0.505, shapeName=dyn_libB_Base_l_capsule, rotation=<rot 0 -90 0>}]
 dyn_libB_l [{shapeType=Box, extentX=5.4892, extentY=18.9232, extentZ=9.5367, center=<4.9211 0 1.0041>, rotation=<rot -1.9914 -90 -12.9387>}]
 dyn_libC_Base_l [{shapeType=Capsule, length=0.505, radius=0.505, shapeName=dyn_libC_Base_l_capsule, rotation=<rot 0 -90 0>}]
 dyn_libC_l [{shapeType=Box, extentX=5.0543, extentY=19.0668, extentZ=10.1809, center=<5.4215 0 0.9223>, rotation=<rot -1.7485 -89.9999 -9.0379>}]
 dyn_libD_Base_l [{shapeType=Capsule, length=0.505, radius=0.505, shapeName=dyn_libD_Base_l_capsule, rotation=<rot 0 -90 0>}]
 dyn_libD_l [{shapeType=Box, extentX=9.636, extentY=18.9525, extentZ=9.9478, center=<5 -1.5 0.3878>, rotation=<rot 0.4728 -89.9996 -4.3757>}]
 dyn_ribA_Base_r [{shapeType=Capsule, length=0.505, radius=0.505, shapeName=dyn_ribA_Base_r_capsule, rotation=<rot 0 -90 0>}]
 dyn_ribA_r [{shapeType=Box, extentX=5.2282, extentY=18.0942, extentZ=9.535, center=<3.9534 0 1.1635>, rotation=<rot -5.065 -90 -17.9054>}]
 dyn_ribB_Base_r [{shapeType=Capsule, length=0.505, radius=0.505, shapeName=dyn_ribB_Base_r_capsule, rotation=<rot 0 -90 0>}]
 dyn_ribB_r [{shapeType=Box, extentX=5.4892, extentY=18.9232, extentZ=9.5367, center=<4.9211 0 1.0041>, rotation=<rot -1.9914 -90 -12.9387>}]
 dyn_ribC_Base_r [{shapeType=Capsule, length=0.505, radius=0.505, shapeName=dyn_ribC_Base_r_capsule, rotation=<rot 0 -90 0>}]
 dyn_ribC_r [{shapeType=Box, extentX=5.0543, extentY=19.0668, extentZ=10.1809, center=<5.4215 0 0.9223>, rotation=<rot -1.7485 -89.9999 -9.0379>}]
 dyn_ribD_Base_r [{shapeType=Capsule, length=0.505, radius=0.505, shapeName=dyn_ribD_Base_r_capsule, rotation=<rot 0 -90 0>}]
 dyn_ribD_r [{shapeType=Box, extentX=9.636, extentY=18.9525, extentZ=9.9478, center=<4.9509 1.4148 0.3878>, rotation=<rot 0.4728 -89.9996 -4.3757>}]
 spine_05 [{
   shapeType=Capsule
   length=20.1704
   radius=16.1253
   shapeName=spine_05_capsule
   center=<9.5551 1.766 -0.0999>
   rotation=<rot 0 14.8353 -89.4876>
  }]
}
constraints [
 {
  bone1Name=dyn_ribA_r
  bone2Name=dyn_ribA_Base_r
  swing1LimitDegrees=10
  swing2LimitDegrees=4
  swing2Motion=Locked
  twistLimitDegrees=4
  twistMotion=Locked
 }
 {
  bone1Name=dyn_ribB_r
  bone2Name=dyn_ribB_Base_r
  swing1LimitDegrees=15
  swing2LimitDegrees=4
  swing2Motion=Locked
  twistLimitDegrees=4
  twistMotion=Locked
 }
 {
  bone1Name=dyn_ribC_r
  bone2Name=dyn_ribC_Base_r
  swing1LimitDegrees=20
  swing2LimitDegrees=4
  swing2Motion=Locked
  twistLimitDegrees=4
  twistMotion=Locked
 }
 {
  bone1Name=dyn_ribD_r
  bone2Name=dyn_ribD_Base_r
  swing1LimitDegrees=20
  swing2LimitDegrees=4
  swing2Motion=Locked
  twistLimitDegrees=4
  twistMotion=Locked
 }
 {
  bone1Name=dyn_libD_l
  bone2Name=dyn_libD_Base_l
  swing1LimitDegrees=20
  swing2LimitDegrees=4
  swing2Motion=Locked
  twistLimitDegrees=4
  twistMotion=Locked
 }
 {
  bone1Name=dyn_libC_l
  bone2Name=dyn_libC_Base_l
  swing1LimitDegrees=20
  swing2LimitDegrees=4
  swing2Motion=Locked
  twistLimitDegrees=4
  twistMotion=Locked
 }
 {
  bone1Name=dyn_libB_l
  bone2Name=dyn_libB_Base_l
  swing1LimitDegrees=15
  swing2LimitDegrees=4
  swing2Motion=Locked
  twistLimitDegrees=4
  twistMotion=Locked
 }
 {
  bone1Name=dyn_libA_l
  bone2Name=dyn_libA_Base_l
  swing1LimitDegrees=10
  swing2LimitDegrees=4
  swing2Motion=Locked
  twistLimitDegrees=4
  twistMotion=Locked
 }
]
